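# Material function for a cape accessory that outputs a wind-animation world position offset by summing two rotations.
# A wiggle rotates around the component pivot on a local-space axis parameter, its angle driven by two summed, time-offset sine waves.
# A slower overall sway rotates around actor position plus a center parameter, with a sine-driven angle and time-varying axis.
# A shading-path switch zeroes the wiggle on mobile, leaving only the sway.
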MaterialFunction /Game/Accessories/FORT_Capes/M_LRG_Donut_Cup_Generic_Cape/Materials/MF_WindAnimationWPO
graph {nodes [
  n1 = Transform(None=VectorParameter() {group=WindAnimation, parameterName=Wiggle_RotaionAxis, defaultValue=<rgba 0 0 1 1>}.RGB) {transformSourceType=TRANSFORMSOURCE_Local}
  n2 = Time()
  n3 = Constant() {r=0.5}
  n4 = Add(A=@n2, B=Multiply(A=@n3, B=ScalarParameter() {defaultValue=-3, group=WindAnimation, parameterName=Wiggle_OffsetVariation}))
  n5 = ScalarParameter() {defaultValue=0.4, group=WindAnimation, parameterName=Wiggle_WindRotation_Speed_1}
  n6 = Add(A=Multiply(A=@n3, B=ScalarParameter() {defaultValue=-1, group=WindAnimation, parameterName=Wiggle_OffsetVariation_1}), B=@n2)
  n7 = ScalarParameter() {defaultValue=0.3, group=WindAnimation, parameterName=Wiggle_WindRotation_Speed}
  n8 = Add(A=Sine(None=Multiply(A=@n4, B=@n5)), B=Sine(None=Multiply(A=@n6, B=@n7)))
  n9 = ScalarParameter() {defaultValue=0.05, group=WindAnimation, parameterName=Wiggle_WindRotation_strength}
  n10 = WorldPosition() {worldPositionShaderOffset=WPT_ExcludeAllShaderOffsets}
  n11 = Add(
   A=Constant3Vector()
   B=RotateAboutAxis(
    NormalizedRotationAxis=Normalize(VectorInput=@n1)
    RotationAngle=Multiply(A=@n8, B=@n9)
    PivotPoint=MaterialFunctionCall() {materialFunction=&ComponentPivotLocation}.Result
    Position=@n10.XYZ
   )
  )
  n12 = Constant3Vector()
  n13 = Multiply(A=Time(), B=ScalarParameter() {defaultValue=0.1714, group=WindAnimation, parameterName=OverallRotationSpeed_2})
  n14 = Sine(None=@n13)
  n15 = Cosine(None=@n13)
  n16 = Normalize(VectorInput=AppendVector(A=Sine(None=@n13), B=AppendVector(A=@n14, B=@n15)))
  n17 = Time()
  n18 = ScalarParameter() {defaultValue=0.3, group=WindAnimation, parameterName=OverallRotationSpeed}
  n19 = Multiply(A=Sine(None=Multiply(A=@n17, B=@n18)), B=ScalarParameter() {defaultValue=0.01, group=WindAnimation, parameterName=OverallRotation})
  n20 = Add(A=VectorParameter() {parameterName=OverallRotationCenter, defaultValue=<rgba 0 0 336.121 1>}.RGB, B=ActorPositionWS())
  FunctionOutput(None=Add(
    A=ShadingPathSwitch(Default=@n11, Mobile=@n12) {desc=Migrated from FeatureLevelSwitch}
    B=RotateAboutAxis(NormalizedRotationAxis=@n16, RotationAngle=@n19, PivotPoint=@n20, Position=@n10.XYZ)
   ))
 ]}
