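# Material function that adds animated cape wind to incoming material attributes, modifying Normal and World Position Offset.
# Pre-skinned position, normalized by bounds, yields height-blended tiled, panned UVs for ripple normal and displacement textures, with sine-pulsed strength.
# Local Y ripple displacement joins the shared MF_WindAnimationWPO offset.
# Height gradients attenuate both effects; green vertex color masks the offset and optionally flattens wind normals before angle-corrected blending.
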MaterialFunction /Game/Characters/Player/Male/Medium/Bodies/M_MED_Plum_Retro/Skin/Plum_Retro_Reactive/Materials/MF_CapeWind
graph {nodes [
  n1 = GetMaterialAttributes(None=FunctionInput() {inputName=Material, inputType=FunctionInput_MaterialAttributes}) {attributeGetTypes [#1, #2]}
  n2 = NamedRerouteDeclaration() {name=Pre Skinned Local Postition, nodeColor=<rgba 1 0.5156 0 1>}
  n3 = PreSkinnedLocalBounds()
  n4 = Divide(A=Subtract(A=FunctionInput(Preview=@n2) {inputName=Preskinned_Input}, B=@n3.Min), B=@n3.Extents)
  n5 = ComponentMask(None=@n4) {g=true, r=true}
  n6 = ComponentMask(None=VectorParameter() {group=CapeWind, parameterName=RippleTiling, defaultValue=<rgba 1 1 0 0>}.RGB) {g=true, r=true}
  n7 = ComponentMask(None=VectorParameter() {group=CapeWind, parameterName=RippleTilingTop, defaultValue=<rgba 5 1 0 0>}.RGB) {g=true, r=true}
  n8 = ComponentMask(None=@n4) {b=true}
  n9 = ScalarParameter() {defaultValue=0.45, group=CapeWind, parameterName=VerticalGratientSSMin}
  n10 = ScalarParameter() {defaultValue=0.5, group=CapeWind, parameterName=VerticalGratientSSMax}
  n11 = Panner(
   Coordinate=LinearInterpolate(
    A=MaterialFunctionCall(UVs=@n5, Texture Scale=@n6) {materialFunction=&ScaleUVsByCenter}.UVs
    B=MaterialFunctionCall(UVs=@n5, Texture Scale=@n7) {materialFunction=&ScaleUVsByCenter}.UVs
    Alpha=MaterialFunctionCall(Alpha=@n8, Min=@n9, Max=@n10) {materialFunction=&SmoothStep}.Result
   )
   Speed=ComponentMask(None=VectorParameter() {group=CapeWind, parameterName=RippleSpeed, defaultValue=<rgba 0 1 0 0>}.RGB) {g=true, r=true}
  )
  n12 = MaterialFunctionCall(Float3=TextureSample(UVs=@n11) {samplerType=SAMPLERTYPE_Normal, texture=/Game/Global/Textures/Default/Noise/T_VerticalRipples_N}.RGB) {materialFunction=&BreakOutFloat3Components}
  n13 = MaterialFunctionCall(X=@n12.R, Y=Power(Base=@n12.G, Exp=ScalarParameter() {defaultValue=1, group=CapeWind, parameterName=RippleNormalPow})) {materialFunction=&MakeFloat2}
  n14 = Multiply(A=Time(), B=ScalarParameter() {defaultValue=0.2, group=CapeWind, parameterName=TimeRippleAttenSpeed})
  n15 = Max(A=ConstantBiasScale(None=Sine(None=@n14)), B=ScalarParameter() {defaultValue=0.15, group=CapeWind, parameterName=TimeRippleAttenMin}) {constB=0.15}
  n16 = Multiply(A=ScalarParameter() {defaultValue=1, group=CapeWind, parameterName=RippleNormalIntensity}, B=@n15)
  n17 = Normalize(VectorInput=AppendVector(A=Multiply(A=@n13.Result, B=@n16), B=@n12.B))
  n18 = VertexColor()
  n19 = MaterialFunctionCall(
   Normal=@n17
   Flatness=LinearInterpolate(A=ScalarParameter() {defaultValue=1, parameterName=MaxNormalFlatness}, B=ScalarParameter() {parameterName=MinNormalFlatness}, Alpha=@n18.G)
  ) {materialFunction=&FlattenNormal}
  n20 = Normalize(VectorInput=MaterialFunctionCall(
    BaseNormal=StaticSwitchParameter(True=@n19.Result, False=@n17) {group=CapeWind, parameterName=FlattenWindNormalsWithVertexColors}
    AdditionalNormal=@n1.Normal
   ) {materialFunction=&BlendAngleCorrectedNormals}.Result)
  n21 = Saturate(None=@n8)
  n22 = Saturate(None=Subtract(A=OneMinus(None=@n21), B=ScalarParameter() {group=CapeWind, parameterName=HeightAttenuationOffset}))
  n23 = Power(Base=@n22, Exp=ScalarParameter() {defaultValue=1, group=CapeWind, parameterName=NormalsAttenuationPow})
  n24 = Multiply(
   A=MaterialFunctionCall() {materialFunction=/Game/Accessories/FORT_Capes/M_LRG_Donut_Cup_Generic_Cape/Materials/MF_WindAnimationWPO}.Result
   B=ScalarParameter() {group=CapeWind, parameterName=WindWPOIntensity}
  )
  n25 = Constant()
  n26 = TextureSample(UVs=@n11) {samplerType=SAMPLERTYPE_Masks, texture=/Game/Global/Textures/Default/Noise/T_VerticalRipples_D}
  n27 = Multiply(
   A=Transform(None=MaterialFunctionCall(R=@n25, G=@n26.G, B=@n25) {materialFunction=&AppendMany}.RGB) {transformSourceType=TRANSFORMSOURCE_Local}
   B=Multiply(A=ScalarParameter() {group=CapeWind, parameterName=RippleWPOIntensity}, B=@n15)
  )
  n28 = ScalarParameter() {defaultValue=1, group=CapeWind, parameterName=HeightAttenuationPow}
  n29 = Add(A=@n1.WorldPositionOffset, B=Multiply(A=Add(A=@n24, B=@n27), B=Power(Base=@n22, Exp=@n28)))
  FunctionOutput(None=SetMaterialAttributes!(
    MaterialAttributes=@n1.MaterialAttributes
    Normal=LinearInterpolate(A=@n1.Normal, B=@n20, Alpha=@n23)
    WorldPositionOffset=LinearInterpolate(A=@n1.WorldPositionOffset, B=@n29, Alpha=@n18.G)
   ) {attributeSetTypes [#1, #2]})
 ]}
